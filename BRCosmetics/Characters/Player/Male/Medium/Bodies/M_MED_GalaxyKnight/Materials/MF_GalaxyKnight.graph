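# Material function that modifies incoming material attributes for the Galaxy Knight character's glowing galaxy look.
# Fresnel and glow terms shape subsurface color, metallic, roughness and edge-darkened emissive.
# Camera-vector triplanar stars twinkle through scrolling biplanar cell noise, and a screen-aligned sphere mask forms a chest glow with a debug switch.
# A static switch adds a backpack core with stars and a radial swirl, using a derivative-based UV seam fix.
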MaterialFunction /BRCosmetics/Characters/Player/Male/Medium/Bodies/M_MED_GalaxyKnight/Materials/MF_GalaxyKnight
graph {nodes [
  n1 = NamedRerouteUsage()
  n2 = FunctionInput() {inputType=FunctionInput_MaterialAttributes}
  n3 = GetMaterialAttributes(None=@n2) {attributeGetTypes [
    #1
    #2
    #3
    #4
    #5
   ]}
  n4 = Abs(None=ComponentMask(None=NamedRerouteUsage()) {r=true})
  n5 = OneMinus(None=Saturate(None=Divide(A=@n4) {constB=18}))
  n6 = ScalarParameter() {defaultValue=2, parameterName=EdgeDarkeningPower}
  n7 = Add(A=NamedRerouteUsage(), B=Multiply(A=@n3.EmissiveColor, B=Power(Base=@n5, Exp=@n6)))
  n8 = SetMaterialAttributes!(
   MaterialAttributes=@n2
   SubsurfaceColor=NamedRerouteUsage()
   Metallic=NamedRerouteUsage()
   Roughness=NamedRerouteUsage()
   Opacity=ScalarParameter() {defaultValue=0.1, group=GalaxyKnight, parameterName=GK_SubsurfaceOpacity}
   EmissiveColor=StaticSwitchParameter(True=Add(A=@n1, B=@n7), False=@n7) {group=GalaxyKnight, parameterName=BackpackCore}
   Normal
   BaseColor=StaticSwitchParameter(True=NamedRerouteUsage(), False=@n3.BaseColor) {parameterName=DebugChestGlow}
  ) {attributeSetTypes [
    #6
    #2
    #1
    #7
    #4
    #3
    #5
   ]}
  FunctionOutput(None=@n8)
  Multiply(
   A=MaterialFunctionCall!(CameraVector) {materialFunction=/Engine/Functions/Engine_MaterialFunctions01/Texturing/CameraVectorToLatLongUV}.Uvs
   B=ComponentMask(None=VectorParameter() {parameterName=GK_StarLatLongUVTransform, defaultValue=<rgba 18 9 0 1>}.RGB) {g=true, r=true}
  ) {constB=4}
  n9 = LinearInterpolate(Alpha=Fresnel(ExponentIn=ScalarParameter() {defaultValue=0.5, group=GalaxyKnight, parameterName=GK_InnerGlowFresnelExp}, Normal=PixelNormalWS()) {baseReflectFraction=0.05, exponent=1}) {constA=1, constB=-0.5}
  NamedRerouteDeclaration(None=Power(Base=Saturate(None=@n9))) {name=InnerFresnelMask, nodeColor=<rgba 0 0.0625 1 1>}
  n10 = Multiply(
   A=Power(Base=NamedRerouteUsage()) {constExponent=3}
   B=ScalarParameter() {defaultValue=1, group=GalaxyKnight, parameterName=GK_FresnelGlowStrength}
  )
  n11 = NamedRerouteUsage()
  NamedRerouteDeclaration(None=Multiply(
    A=Add(A=@n10, B=@n11)
    B=VectorParameter() {group=GalaxyKnight, parameterName=GK_SubsurfaceColor, defaultValue=<rgba 0.3228 0.1047 0.3854 1>}.RGB
   )) {name=SubsurfaceColor, nodeColor=<rgba 0.3359 0 1 1>}
  NamedRerouteDeclaration(None=@n3.EmissiveColor) {name=Emissive_Pre, nodeColor=<rgba 0.1484 1 0 1>}
  n12 = ConstantBiasScale(None=TextureCoordinate()) {bias=-0.5, scale=2}
  n13 = Constant2Vector() {g=0.3, r=0.4}
  n14 = MaterialFunctionCall(Vector or UVs=Multiply(A=Add(A=@n12, B=@n13), B=ScalarParameter() {defaultValue=3.8, parameterName=GK_BackpackSwirlTIling}) {constB=3.5}) {materialFunction=&VectorToRadialValue}
  MaterialExpressionFrac_5 = Frac(None=@n14.Vector Converted to Angle)
  MaterialExpressionSubtract_3 = Subtract(A=Frac(None=Add(A=@n14.Vector Converted to Angle) {constB=0.5})) {constB=0.5}
  n15 = Custom!(u0=@MaterialExpressionFrac_5, u1=@MaterialExpressionSubtract_3) {
   code="return ((fwidth(u0) < fwidth(u1)-0.001)? u0:u1);"
   description=FixPerPixelCylidricalUV
   outputType=CMOT_Float1
   inputs [
    {
     inputName=u0
     input {expression=&self:MaterialExpressionFrac_5}
    }
    {
     inputName=u1
     input {expression=&self:MaterialExpressionSubtract_3}
    }
   ]
  }
  n16 = Time()
  n17 = ScalarParameter() {defaultValue=-0.7, parameterName=GK_BackpackSwirlSpeed}
  n18 = Add(A=ComponentMask(None=@n15) {r=true}, B=Frac(None=Multiply(A=@n16, B=@n17) {constB=-1}))
  n19 = AppendVector(A=@n18, B=Add(A=@n14.Linear Distance, B=@n18))
  n20 = VectorParameter() {group=GalaxyKnight, parameterName=GK_BackpackSwirlColor2, defaultValue=<rgba 0.2401 0.0451 0.7865 1>}
  n21 = VectorParameter() {group=GalaxyKnight, parameterName=GK_BackpackSwirlColor1, defaultValue=<rgba 0.979 0 1 1>}
  n22 = Power(Base=SphereMask(A=TextureCoordinate(), B=Constant2Vector() {g=0.31, r=0.28}) {attenuationRadius=1, hardnessPercent=0}) {constExponent=7}
  n23 = TextureSample() {
   samplerType=SAMPLERTYPE_Masks
   texture=/Game/Accessories/FORT_Backpacks/Backpack_M_MED_GalaxyKnight/Textures/T_M_MED_GalaxyKnight_Backpack_FX
  }
  n24 = Multiply(
   A=Multiply(
    A=TextureSample(UVs=@n19) {texture=/Game/Global/Textures/Default/Masks_and_Gradients/CylGradient}.R
    B=LinearInterpolate(A=@n20.RGB, B=@n21.RGB, Alpha=@n22)
   )
   B=@n23.G
  )
  n25 = OneMinus(None=Fresnel() {exponent=0.5})
  n26 = Power(Base=SphereMask(A=TextureCoordinate(), B=Constant2Vector() {g=0.31, r=0.28}) {attenuationRadius=0.3, hardnessPercent=0}) {constExponent=7}
  n27 = Multiply(
   A=Multiply(A=@n26, B=VectorParameter() {group=GalaxyKnight, parameterName=GK_BackpackCoreColor, defaultValue=<rgba 0.979 0 1 1>}.RGB)
   B=ScalarParameter() {defaultValue=5, parameterName=GK_BackpackCoreEmsMult}
  )
  n28 = TextureSample(
   UVs=Panner(Coordinate=TextureCoordinate() {uTiling=3, vTiling=3}) {speedY=0.2}
   Tex=TextureObject() {texture=/Game/Weapons/WeaponSkins/Wraps/Textures/ShootingStars/ShootingStars_BackgroundStars_D}
  ) {samplerSource=SSM_Wrap_WorldGroupSettings}
  n29 = VertexInterpolator(VS=PreSkinnedPosition())
  n30 = Multiply(A=Time()) {constB=5}
  n31 = ComponentMask(None=MaterialFunctionCall!(
    ColorTex=TextureObject() {samplerType=SAMPLERTYPE_Grayscale, texture=/BRCosmetics/Weapons/FORT_Melee/Pickaxe_Cube_Ninja_Male/FX/T_CellNoise}
    NormalTex
    Pos=Add(A=@n29.PS, B=@n30)
    Nrm
    ColorUVScale=Constant() {r=75}
    ColorBlendFactor
    NormalStrength=Constant() {r=1}
    NormalUVScale
    NormalBlendFactor
    Local Space Inputs
   ) {materialFunction=/Game/Global/MaterialFunctions/MF_BiPlanar}.Color) {r=true}
  n32 = Multiply(A=Multiply(A=Multiply(A=@n26, B=@n28.R), B=ScalarParameter() {defaultValue=10, parameterName=GK_BackpackStarEmsMult}), B=@n31)
  NamedRerouteDeclaration(None=Add(A=Multiply(A=@n24, B=@n25), B=Add(A=@n27, B=@n32))) {name=BackpackCore, nodeColor=<rgba 0.2891 1 0 1>}
  n33 = VectorParameter() {parameterName=GK_ChestGlowLocation, defaultValue=<rgba 0 0 0 1>}
  n34 = Multiply(A=NamedRerouteUsage()) {constB=5}
  n35 = ViewSize()
  n36 = Multiply(
   A=MaterialFunctionCall!(Center Point World Position=Add(A=@n33.RGB, B=@n34)) {materialFunction=/Engine/Functions/Engine_MaterialFunctions02/UVs/WorldSpaceAlignedScreenCoordinates}.Offset and Scaled UVS
   B=Divide(A=ComponentMask(None=@n35) {r=true}, B=Constant() {r=1080})
  )
  n37 = VectorParameter() {group=GalaxyKnight, parameterName=GK_ChestGlowTransform, defaultValue=<rgba 1 0.6 0 1>}
  n38 = Multiply(A=Multiply(A=Divide(A=@n36) {constB=0.5}, B=ComponentMask(None=@n37.RGB) {g=true, r=true})) {constB=100}
  n39 = Constant2Vector()
  n40 = ScalarParameter() {defaultValue=0.75, group=GalaxyKnight, parameterName=GK_ChestGlowRadius}
  n41 = Multiply(
   A=Power(
    Base=SphereMask(A=@n38, B=@n39, Radius=@n40) {hardnessPercent=0}
    Exp=ScalarParameter() {defaultValue=3, group=GalaxyKnight, parameterName=GK_ChestGlowPower}
   ) {constExponent=5}
   B=ScalarParameter() {defaultValue=0.75, group=GalaxyKnight, parameterName=GK_ChestGlowStrength}
  )
  NamedRerouteDeclaration(None=@n41) {name=Debug, nodeColor=<rgba 1 0.3984 0 1>}
  n42 = NamedRerouteUsage()
  NamedRerouteDeclaration(None=LinearInterpolate(A=@n3.Metallic, B=ScalarParameter() {group=GalaxyKnight, parameterName=GK_InnerMetalness}, Alpha=@n42)) {name=Metal, nodeColor=<rgba 1 0.5156 0 1>}
  NamedRerouteDeclaration(None=LinearInterpolate(A=@n3.Roughness, B=ScalarParameter() {defaultValue=1, group=GalaxyKnight, parameterName=GK_InnerRoughness}, Alpha=@n42)) {name=Rough, nodeColor=<rgba 0.8047 0 1 1>}
  n43 = PreSkinnedPosition()
  NamedRerouteDeclaration(None=ShaderStageSwitch(PixelShader=VertexInterpolator(VS=@n43).PS, VertexShader=@n43)) {name=PreskinnedPos, nodeColor=<rgba 1 0 0.1641 1>}
  n44 = PreSkinnedNormal()
  NamedRerouteDeclaration(None=ShaderStageSwitch(PixelShader=VertexInterpolator(VS=@n44).PS, VertexShader=@n44)) {name=PreskinnedN, nodeColor=<rgba 0.6188 0 1 1>}
  NamedRerouteDeclaration(None=@n41) {name=ChestGlow, nodeColor=<rgba 0 1 0.3906 1>}
  n45 = TextureObject() {texture=/Game/Characters/Player/Male/Medium/Bodies/M_MED_HighTower_Tapas/Textures/T_Hightower_Tapas_Star}
  n46 = ScalarParameter() {defaultValue=1.5, group=GalaxyKnight, parameterName=GK_StarTiling}
  n47 = Constant() {r=1}
  n48 = LinearInterpolate(Alpha=Fresnel(ExponentIn=ScalarParameter() {defaultValue=0.5, group=GalaxyKnight, parameterName=GK_StarFresnelExp}, Normal=VertexNormalWS()) {baseReflectFraction=0.05, exponent=1}) {constA=10, constB=-10}
  n49 = Multiply(
   A=Add(
    A=MaterialFunctionCall!(CameraVector, Texture=@n45, Tiling=@n46, Offset=@n47, AxisFadeContrast) {materialFunction=/Engine/Functions/Engine_MaterialFunctions01/Texturing/TriplanarCameraVector}.Result
    B=MaterialFunctionCall(Texture=@n45, Tiling=@n46) {materialFunction=&TriplanarCameraVector}.Result
   )
   B=OneMinus(None=Saturate(None=@n48))
  )
  NamedRerouteDeclaration(None=Multiply(A=Multiply(A=@n49, B=@n31), B=ScalarParameter() {defaultValue=15, group=GalaxyKnight, parameterName=GK_StarEms})) {name=Star, nodeColor=<rgba 0 1 0.6484 1>}
  TransformPosition(None=@n33.RGB) {transformType=TRANSFORMPOSSOURCE_World}
  VertexNormalWS()
 ]}
