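# Parent material for the male medium GalaxyKnight body, a subsurface-shaded skeletal mesh surface built on material attributes.
# It chains shared character functions, running BaseFN_CharacterFunction_PreFX into CharacterSkinFX, then MF_GalaxyKnight, then BaseFN_CharacterFunction_PostFX.
# PreFX outputs also feed the auxiliary inputs of PostFX directly, bypassing the effect stages.
# The skin-specific look lives in MF_GalaxyKnight, so this graph holds only the wiring.
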Material /BRCosmetics/Characters/Player/Male/Medium/Bodies/M_MED_GalaxyKnight/Materials/M_MED_GalaxyKnight_Parent
graph {
 properties {
  bUseMaterialAttributes=true
  bUsedWithMeshDeformer=true
  bUsedWithMorphTargets=true
  bUsedWithSkeletalMesh=true
  pixelDepthOffsetMode=PDOM_Legacy
  refractionMethod=RM_IndexOfRefraction
  shadingModel=MSM_Subsurface
  usedShadingModels=Subsurface
 }
 nodes [
  n1 = MaterialFunctionCall() {materialFunction=&BaseFN_CharacterFunction_PreFX}
  MaterialFunctionCall(
   AO=@n1.AO
   Distance Blend=@n1.AO
   Fresnel=@n1.AO
   Gmap=@n1.AO
   Pre Skinned Local Position=@n1.AO
   Material=MaterialFunctionCall!(In=MaterialFunctionCall(
     PreskinnedPosition=@n1.Pre skinned Local Position
     Gmap=@n1.Pre skinned Local Position
     MaterialAttributes=@n1.Pre skinned Local Position
     AO=@n1.Pre skinned Local Position
    ) {materialFunction=&CharacterSkinFX}.Result) {materialFunction=&/MF_GalaxyKnight}.Result
  ) {materialFunction=&BaseFN_CharacterFunction_PostFX}
 ]
}
