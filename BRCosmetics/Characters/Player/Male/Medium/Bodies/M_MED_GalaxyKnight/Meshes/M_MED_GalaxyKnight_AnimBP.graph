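# Animation Blueprint for the Galaxy Knight medium male cosmetic body, posing it from the attached parent mesh and driving its chest glow material.
# The AnimGraph copies the parent's pose and runs it through the body's Control Rig; deform toggle and offset variables hold per-joint defaults.
# After animation evaluation, a zero-length delay precedes writing the ChestGlowSocket location into the GK_ChestGlowLocation material vector parameter.
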AnimBlueprint /BRCosmetics/Characters/Player/Male/Medium/Bodies/M_MED_GalaxyKnight/Meshes/M_MED_GalaxyKnight_AnimBP
target_skeleton=&/M_MED_GalaxyKnight_Skeleton
graphs [
 AnimGraph {nodes [
   n1 = Copy Pose From Mesh() {bUseAttachedParent=true}
   n2 = Control Rig(Source=Component To Local(ComponentPose=Local To Component(LocalPose=@n1))) {
    alphaBoolBlend {blendOption=Linear}
    controlRigAssetReference {blueprintRigClass=&/M_MED_GalaxyKnight_ControlRig.M_MED_GalaxyKnight_ControlRig_C}
   }
   Output Pose(Result=@n2) {layerGroup=None}
  ]}
 EventGraph {nodes [
   Event BlueprintPostEvaluateAnimation() {is=title K2Node_Event}
   n1 = GetPartSkeletalMeshComponent()
   > Is Valid(InputObject=@n1)
   > Delay(Duration=0)
   > SetVectorParameterValueOnMaterials!(
    execute
    self:Mesh Component Object Reference=@n1
    ParameterName:Name=GK_ChestGlowLocation
    ParameterValue:Vector=GetSocketLocation!(self:Scene Component Object Reference=@n1, InSocketName:Name=ChestGlowSocket) -> (ReturnValue:Vector)
   ) -> (then)
   Event BlueprintUpdateAnimation()
  ]}
]
variables [
 deform_bool {
  type=FN_DeformRig_Bool_Struct
  default {bicep=false, elbow=true, glute=true, groin=true, knee=true, lat=false, pec=true, upperarm=true, wrist=true}
 }
 deform_values {
  type=FN_DeformRig_Value_Struct
  default {
   bicep {location=<-3 4 0>, rotation=<rot 0 0 0>, scale=<1 1 1>}
   elbow_in {location=<0 9 0>, rotation=<rot 0 0 0>, scale=<0.5 1 1>}
   elbow_out {location=<0 -2 0>, rotation=<rot 0 0 0>, scale=<0.75 1 1>}
   glute {location=<-7 0 0>, rotation=<rot 0 0 0>, scale=<1 1.5 1>}
   groin {location=<-3 12 0>, rotation=<rot 0 0 0>, scale=<0.5 1 1>}
   knee_in {location=<0 -12 0>, rotation=<rot 0 0 0>, scale=<0.5 1 1>}
   knee_out {location=<0 3 0>, rotation=<rot 0 0 0>, scale=<1.5 1 1.5>}
   lat {location=<0 0 10>, rotation=<rot 0 0 0>, scale=<1 1 1>}
   pec {location=<0 2 3>, rotation=<rot 0 0 0>, scale=<1 1 1>}
   upperarm_bk {location=<0 -4 0>, rotation=<rot 0 0 0>, scale=<1 1 1>}
   upperarm_fwd_bk {location=<0 -6 0>, rotation=<rot 0 0 0>, scale=<1 1 1>}
   upperarm_fwd_fr {location=<0 10 0>, rotation=<rot 0 0 0>, scale=<0.5 1 1>}
   upperarm_up {location=<0 0 8>, rotation=<rot 0 0 0>, scale=<0.7 1 1>}
   wrist_dn {location=<0 -6 0>, rotation=<rot 0 0 0>, scale=<0.7 1 1>}
   wrist_up {location=<0 9 0>, rotation=<rot 0 0 0>, scale=<0.7 1 1>}
  }
 }
]
